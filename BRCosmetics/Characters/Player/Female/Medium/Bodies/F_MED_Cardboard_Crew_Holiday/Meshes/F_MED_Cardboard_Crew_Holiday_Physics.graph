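# Physics asset for the holiday Cardboard Crew female medium body cosmetic, defining bodies for the costume's dangling balls, front ribbon, torso boxes and flaps.
# Capsules on pelvis, spine and thighs form the body, with flap and box bases linked to spine_03 by free constraints.
# Torso flaps act as single-axis hinges with twist locked.
# Balls and ribbon are limited-swing, low-twist chains, with the ribbon tied to both the front flap and its base.
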PhysicsAsset /BRCosmetics/Characters/Player/Female/Medium/Bodies/F_MED_Cardboard_Crew_Holiday/Meshes/F_MED_Cardboard_Crew_Holiday_Physics
bodies {
 dyn_ball_bck_01 [{
   shapeType=Capsule
   length=2.9898
   radius=1.2582
   shapeName=dyn_ball_bck_01_capsule
   center=<2.5311 0 0>
   rotation=<rot 90 -166.7373 -166.7374>
  }]
 dyn_ball_bck_02 [{shapeType=Capsule, length=0.8045, radius=4.2375, shapeName=dyn_ball_bck_02_capsule, center=<6.3411 1.4861 0.2261>}]
 dyn_ball_bck_Base [{shapeType=Capsule, length=1, radius=1, shapeName=dyn_ball_bck_Base_capsule}]
 dyn_ball_fr_01 [{shapeType=Capsule, length=2.9898, radius=1.2582, shapeName=dyn_ball_fr_01_capsule, center=<2.5311 0 0>, rotation=<rot 90 -90 -90>}]
 dyn_ball_fr_02 [{shapeType=Capsule, length=0.8887, radius=4.2778, shapeName=dyn_ball_fr_02_capsule, center=<6.7397 0 0>}]
 dyn_ball_fr_Base [{shapeType=Capsule, length=1, radius=1, shapeName=dyn_ball_fr_Base_capsule}]
 dyn_ribbon_01_fr [{
   shapeType=Capsule
   length=5.174
   radius=2.0141
   shapeName=dyn_ribbon_01_fr_capsule
   center=<1.7388 0 0>
   rotation=<rot -90 160.5288 -160.5288>
  }]
 dyn_ribbon_02_fr [{shapeType=Capsule, length=6.764, radius=2.7712, shapeName=dyn_ribbon_02_fr_capsule, center=<1.4394 0 0>, rotation=<rot -90 90 -90.0001>}]
 dyn_ribbon_03_fr [{shapeType=Capsule, length=7.0755, radius=2.7712, shapeName=dyn_ribbon_03_fr_capsule, center=<3.0791 0 0>, rotation=<rot -90 90 -90>}]
 dyn_ribbon_Base_fr [{shapeType=Capsule, length=1, radius=1, shapeName=dyn_ribbon_Base_fr_capsule, rotation=<rot -90 120.9442 -120.9443>}]
 dyn_torsoBox_Base_l [{shapeType=Capsule, length=1, radius=1, shapeName=dyn_torsoBox_Base_l_capsule, rotation=<rot 0 -90 0>}]
 dyn_torsoBox_Base_r [{shapeType=Capsule, length=1, radius=1, shapeName=dyn_torsoBox_Base_r_capsule, rotation=<rot 0 -90 0>}]
 dyn_torsoBox_l [{
   shapeType=Box
   extentX=9.944
   extentY=10.8755
   extentZ=9.4239
   shapeName=dyn_torsoBox_l_box
   center=<1.9555 -1.8013 -0.287>
   rotation=<rot 45.0009 0 0>
  }]
 dyn_torsoBox_r [{
   shapeType=Box
   extentX=8.4095
   extentY=11.2993
   extentZ=9.3389
   shapeName=dyn_torsoBox_r_box
   center=<2.387 -1.4648 0.4346>
   rotation=<rot 44.9319 3.9746 2.8093>
  }]
 dyn_torsoFlap_Base_l [{shapeType=Capsule, length=1, radius=1, shapeName=dyn_torsoFlap_Base_l_capsule, rotation=<rot 0 -90 0>}]
 dyn_torsoFlap_Base_r [{shapeType=Capsule, length=1, radius=1, shapeName=dyn_torsoFlap_Base_r_capsule}]
 dyn_torsoFlap_back [{
   shapeType=Capsule
   length=14.6376
   radius=4.383
   shapeName=dyn_torsoFlap_back_capsule
   center=<6.2568 0.9374 -0.0499>
   rotation=<rot -5.4612 -90 4.4096>
  }]
 dyn_torsoFlap_back_Base [{shapeType=Capsule, length=1, radius=1, shapeName=dyn_torsoFlap_back_Base_capsule}]
 dyn_torsoFlap_front [{
   shapeType=Capsule
   length=16.8253
   radius=3.8322
   shapeName=dyn_torsoFlap_front_capsule
   center=<4.3593 0.1303 1.3945>
   rotation=<rot 0.9195 -90 9.5584>
  }]
 dyn_torsoFlap_front_Base [{shapeType=Capsule, length=1, radius=1, shapeName=dyn_torsoFlap_front_Base_capsule}]
 dyn_torsoFlap_l [{
   shapeType=Capsule
   length=16.0811
   radius=4.2759
   shapeName=dyn_torsoFlap_l_capsule
   center=<4.1921 -0.1151 0.7338>
   rotation=<rot -1.3327 -90 7.3053>
  }]
 dyn_torsoFlap_r [{
   shapeType=Capsule
   length=14.7298
   radius=4.0693
   shapeName=dyn_torsoFlap_r_capsule
   center=<4.3093 0.0806 -0.3815>
   rotation=<rot 0.0763 -90 -0.8296>
  }]
 pelvis [{shapeType=Capsule, length=9.3509, radius=10.6464, shapeName=pelvis_capsule, center=<-3.0434 -1.9086 0>}]
 spine_01 [{shapeType=Capsule, length=8.1076, radius=8.6386, shapeName=spine_01_capsule, center=<0.9069 0 0>, rotation=<rot 0 -0.0001 0>}]
 spine_02 [{shapeType=Capsule, length=4.0066, radius=7.06, shapeName=spine_02_capsule, center=<0 1.6752 0>}]
 spine_03 [{shapeType=Capsule, length=7.3441, radius=8.8522, shapeName=spine_03_capsule, center=<0 1.5949 0>, rotation=<rot 90 0 0>}]
 thigh_l [{shapeType=Capsule, length=25.1518, radius=7.2417, shapeName=thigh_l_capsule, center=<-19.7848 0 0>, rotation=<rot 90 0 0>}]
 thigh_r [{shapeType=Capsule, length=28.7879, radius=7.4218, shapeName=thigh_r_capsule, center=<21.9617 0 0>, rotation=<rot -90 180 179.9999>}]
}
constraints [
 {bone1Name=dyn_ball_bck_02, bone2Name=dyn_ball_bck_01, swing1LimitDegrees=45, swing2LimitDegrees=45, twistLimitDegrees=5}
 {bone1Name=dyn_ball_fr_02, bone2Name=dyn_ball_fr_01, swing1LimitDegrees=45, swing2LimitDegrees=45, twistLimitDegrees=5}
 {bone1Name=dyn_ball_fr_01, bone2Name=dyn_ball_fr_Base, swing1LimitDegrees=30, swing2LimitDegrees=30, twistLimitDegrees=5}
 {bone1Name=dyn_ball_bck_01, bone2Name=dyn_ball_bck_Base, swing1LimitDegrees=30, swing2LimitDegrees=30, twistLimitDegrees=5}
 {
  bone1Name=spine_01
  bone2Name=pelvis
  swing1LimitDegrees=45
  swing1Motion=Free
  swing2LimitDegrees=45
  swing2Motion=Free
  twistLimitDegrees=45
  twistMotion=Free
 }
 {
  bone1Name=spine_02
  bone2Name=spine_01
  swing1LimitDegrees=45
  swing1Motion=Free
  swing2LimitDegrees=45
  swing2Motion=Free
  twistLimitDegrees=45
  twistMotion=Free
 }
 {
  bone1Name=spine_03
  bone2Name=spine_02
  swing1LimitDegrees=45
  swing1Motion=Free
  swing2LimitDegrees=45
  swing2Motion=Free
  twistLimitDegrees=45
  twistMotion=Free
 }
 {
  bone1Name=dyn_torsoBox_l
  bone2Name=dyn_torsoBox_Base_l
  swing1LimitDegrees=45
  swing1Motion=Free
  swing2LimitDegrees=45
  swing2Motion=Free
  twistLimitDegrees=45
  twistMotion=Free
 }
 {
  bone1Name=dyn_torsoBox_r
  bone2Name=dyn_torsoBox_Base_r
  swing1LimitDegrees=45
  swing1Motion=Free
  swing2LimitDegrees=45
  swing2Motion=Free
  twistLimitDegrees=45
  twistMotion=Free
 }
 {
  bone1Name=dyn_torsoBox_Base_l
  bone2Name=spine_03
  swing1LimitDegrees=45
  swing1Motion=Free
  swing2LimitDegrees=45
  swing2Motion=Free
  twistLimitDegrees=45
  twistMotion=Free
 }
 {
  bone1Name=dyn_torsoBox_Base_r
  bone2Name=spine_03
  swing1LimitDegrees=45
  swing1Motion=Free
  swing2LimitDegrees=45
  swing2Motion=Free
  twistLimitDegrees=45
  twistMotion=Free
 }
 {
  bone1Name=dyn_torsoFlap_l
  bone2Name=dyn_torsoFlap_Base_l
  swing1LimitDegrees=45
  swing2LimitDegrees=45
  swing2Motion=Locked
  twistLimitDegrees=45
  twistMotion=Locked
 }
 {
  bone1Name=dyn_torsoFlap_r
  bone2Name=dyn_torsoFlap_Base_r
  swing1LimitDegrees=45
  swing2LimitDegrees=45
  swing2Motion=Locked
  twistLimitDegrees=45
  twistMotion=Locked
 }
 {
  bone1Name=dyn_torsoFlap_back
  bone2Name=dyn_torsoFlap_back_Base
  swing1LimitDegrees=30
  swing2LimitDegrees=45
  swing2Motion=Locked
  twistLimitDegrees=45
  twistMotion=Locked
 }
 {
  bone1Name=dyn_torsoFlap_front
  bone2Name=dyn_torsoFlap_front_Base
  swing1LimitDegrees=45
  swing2LimitDegrees=45
  swing2Motion=Locked
  twistLimitDegrees=45
  twistMotion=Locked
 }
 {
  bone1Name=dyn_torsoFlap_front_Base
  bone2Name=spine_03
  swing1LimitDegrees=45
  swing1Motion=Free
  swing2LimitDegrees=45
  swing2Motion=Free
  twistLimitDegrees=45
  twistMotion=Free
 }
 {
  bone1Name=dyn_torsoFlap_back_Base
  bone2Name=spine_03
  swing1LimitDegrees=45
  swing1Motion=Free
  swing2LimitDegrees=45
  swing2Motion=Free
  twistLimitDegrees=45
  twistMotion=Free
 }
 {
  bone1Name=dyn_torsoFlap_Base_r
  bone2Name=spine_03
  swing1LimitDegrees=45
  swing1Motion=Free
  swing2LimitDegrees=45
  swing2Motion=Free
  twistLimitDegrees=45
  twistMotion=Free
 }
 {
  bone1Name=dyn_torsoFlap_Base_l
  bone2Name=spine_03
  swing1LimitDegrees=45
  swing1Motion=Free
  swing2LimitDegrees=45
  swing2Motion=Free
  twistLimitDegrees=45
  twistMotion=Free
 }
 {
  bone1Name=dyn_torsoBox_Base_r
  bone2Name=spine_03
  swing1LimitDegrees=45
  swing1Motion=Free
  swing2LimitDegrees=45
  swing2Motion=Free
  twistLimitDegrees=45
  twistMotion=Free
 }
 {
  bone1Name=dyn_torsoBox_Base_l
  bone2Name=spine_03
  swing1LimitDegrees=45
  swing1Motion=Free
  swing2LimitDegrees=45
  swing2Motion=Free
  twistLimitDegrees=45
  twistMotion=Free
 }
 {
  bone1Name=thigh_l
  bone2Name=pelvis
  swing1LimitDegrees=45
  swing1Motion=Free
  swing2LimitDegrees=45
  swing2Motion=Free
  twistLimitDegrees=45
  twistMotion=Free
 }
 {
  bone1Name=thigh_r
  bone2Name=pelvis
  swing1LimitDegrees=45
  swing1Motion=Free
  swing2LimitDegrees=45
  swing2Motion=Free
  twistLimitDegrees=45
  twistMotion=Free
 }
 {bone1Name=dyn_ribbon_03_fr, bone2Name=dyn_ribbon_02_fr, swing1LimitDegrees=45, swing2LimitDegrees=20, twistLimitDegrees=1}
 {bone1Name=dyn_ribbon_02_fr, bone2Name=dyn_ribbon_01_fr, swing1LimitDegrees=45, swing2LimitDegrees=20, twistLimitDegrees=1}
 {bone1Name=dyn_ribbon_01_fr, bone2Name=dyn_ribbon_Base_fr, swing1LimitDegrees=45, swing2LimitDegrees=20, twistLimitDegrees=1}
 {bone1Name=dyn_ribbon_Base_fr, bone2Name=dyn_torsoFlap_front_Base, swing1LimitDegrees=45, swing2LimitDegrees=45, twistLimitDegrees=1}
 {bone1Name=dyn_ribbon_Base_fr, bone2Name=dyn_torsoFlap_front, swing1LimitDegrees=45, swing2LimitDegrees=20, twistLimitDegrees=1}
 {
  bone1Name=dyn_torsoFlap_front
  bone2Name=dyn_torsoFlap_back
  swing1LimitDegrees=45
  swing1Motion=Free
  swing2LimitDegrees=45
  swing2Motion=Free
  twistLimitDegrees=45
  twistMotion=Free
 }
]
