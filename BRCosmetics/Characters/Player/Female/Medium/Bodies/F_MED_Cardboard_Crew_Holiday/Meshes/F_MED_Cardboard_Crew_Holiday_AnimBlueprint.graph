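# Post-process style Anim Blueprint for the female Cardboard Crew Holiday body, adding secondary motion to a pose copied from the attached parent mesh.
# A Control Rig (crouch-aware, off during emotes) feeds RigidBody, then spring-driven AnimDynamics for bows, ribbons, fragments and laces.
# Additive bone offsets adjust the back flap and ribbon, whose bones simulate individually and as a chain.
# EventGraph updates gravity override, emote state, and backbling flags the AnimGraph never reads.
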AnimBlueprint /BRCosmetics/Characters/Player/Female/Medium/Bodies/F_MED_Cardboard_Crew_Holiday/Meshes/F_MED_Cardboard_Crew_Holiday_AnimBlueprint
target_skeleton=&/F_MED_Cardboard_Crew_Holiday_Skeleton
graphs [
 AnimGraph {nodes [
   n1 = Control Rig!(
    Source=Use cached pose 'SavedPose15805'()
    bAlphaBoolEnabled:bool=Get Not_playing_Emote() -> (Not_playing_Emote:bool)
    isCrouching:bool=Get bIsCrouching()
   ) -> (Pose) {
    alphaInputType=Bool
    alphaBoolBlend {blendOption=Linear}
    controlRigAssetReference {blueprintRigClass=&/F_MED_Cardboard_Crew_Holiday_CtrlRig.F_MED_Cardboard_Crew_Holiday_CtrlRig_C}
   }
   n2 = RigidBody!(ComponentPose=Local To Component(LocalPose=@n1), Alpha, OverrideWorldGravity:Vector=Get Gravity Override(), ExternalForce) -> (Pose) {
    bClampLinearTranslationLimitToRefPose=true
    bOverrideWorldGravity=true
    overridePhysicsAsset=&/F_MED_Cardboard_Crew_Holiday_Physics
    simulationSpace=BaseBoneSpace
    alphaBoolBlend {blendOption=Linear}
    baseBoneRef {boneName=root}
    componentAppliedLinearAccClamp=<3000 3000 3000>
    componentLinearAccScale=<0.4 4 0.4>
    componentLinearVelScale=<0.25 0.25 0.25>
    overrideWorldGravity {z=-980}
   }
   n3 = Anim Dynamics - Bone: dyn_bow_r_01(ComponentPose=@n2, Alpha=1, ExternalForce="0,0,0") {
    angularDampingOverride=0.9
    angularSpringConstant=250
    bAngularSpring=true
    bOverrideAngularDamping=true
    bUseGravityOverride=true
    bUseSphericalLimits=true
    simulationSpace=RootRelative
    alphaBoolBlend {blendOption=Linear}
    boundBone {boneName=dyn_bow_r_01}
    componentAppliedLinearAccClamp=<2000 2000 2000>
    componentLinearAccScale=<0.3 0.3 0.3>
    componentLinearVelScale=<0.3 0.3 0.3>
    gravityOverride {z=-980}
    physicsBodyDefinitions [{
      boundBone {boneName=dyn_bow_r_01}
      boxExtents=<10 10 10>
      constraintSetup {angularLimitsMax=<0 0 10>, angularLimitsMin=<0 -30 -30>, angularTarget=<0 0 0>}
      localJointOffset=<0.5 0 0>
     }]
    relativeSpaceBone {boneName=root}
    sphericalLimits [{limitRadius=8, drivingBone {boneName=thigh_r}, sphereLocalOffset=<22.7068 0.7723 0>}]
   }
   n4 = Anim Dynamics - Bone: dyn_bow_l_01(ComponentPose=@n3, Alpha=1, ExternalForce="0,0,0") {
    angularDampingOverride=0.9
    angularSpringConstant=250
    bAngularSpring=true
    bOverrideAngularDamping=true
    bUseGravityOverride=true
    bUseSphericalLimits=true
    simulationSpace=RootRelative
    alphaBoolBlend {blendOption=Linear}
    boundBone {boneName=dyn_bow_l_01}
    componentAppliedLinearAccClamp=<2000 2000 2000>
    componentLinearAccScale=<0.3 0.3 0.3>
    componentLinearVelScale=<0.3 0.3 0.3>
    gravityOverride {z=-980}
    physicsBodyDefinitions [{
      boundBone {boneName=dyn_bow_l_01}
      boxExtents=<10 10 10>
      constraintSetup {angularLimitsMax=<0 30 10>, angularLimitsMin=<0 0 -30>, angularTarget=<0 0 0>}
      localJointOffset=<0.5 0 0>
     }]
    relativeSpaceBone {boneName=root}
    sphericalLimits [{limitRadius=8, drivingBone {boneName=thigh_r}, sphereLocalOffset=<22.7068 0.7723 -0.3603>}]
   }
   n5 = "Anim Dynamics - Chain: dyn_bow_ribbon_01_r -> dyn_bow_ribbon_02_r"(ComponentPose=@n4, Alpha=1, ExternalForce="0,0,0") {
    angularDampingOverride=0.9
    angularSpringConstant=350
    bAngularSpring=true
    bChain=true
    bOverrideAngularDamping=true
    bUseGravityOverride=true
    bUseSphericalLimits=true
    simulationSpace=RootRelative
    alphaBoolBlend {blendOption=Linear}
    boundBone {boneName=dyn_bow_ribbon_01_r}
    chainEnd {boneName=dyn_bow_ribbon_02_r}
    componentAppliedLinearAccClamp=<2000 2000 2000>
    componentLinearAccScale=<0.3 0.3 0.3>
    componentLinearVelScale=<0.3 0.3 0.3>
    gravityOverride {z=-980}
    physicsBodyDefinitions [
     {
      boundBone {boneName=dyn_bow_ribbon_01_r}
      boxExtents=<20 20 20>
      constraintSetup {angularLimitsMax=<0 10 20>, angularLimitsMin=<0 -10 0>, angularTarget=<0 0 0>}
      localJointOffset=<0.5 0 0>
     }
     {
      boundBone {boneName=dyn_bow_ribbon_02_r}
      boxExtents=<20 20 20>
      constraintSetup {angularLimitsMax=<0 10 20>, angularLimitsMin=<0 -10 0>, angularTarget=<0 0 0>}
      localJointOffset=<0.5 0 0>
     }
    ]
    planarLimits [{drivingBone {boneName=thigh_twist_01_r}, planeTransform {location=<0 -6.6679 -1.192>, rotation=<rot -0.8075 -4.5221 -100.1039>}}]
    relativeSpaceBone {boneName=root}
   }
   n6 = "Anim Dynamics - Chain: dyn_bow_ribbon_01_l -> dyn_bow_ribbon_02_l"(ComponentPose=@n5, Alpha=1, ExternalForce="0,0,0") {
    angularDampingOverride=0.9
    angularSpringConstant=350
    bAngularSpring=true
    bChain=true
    bOverrideAngularDamping=true
    bUseGravityOverride=true
    bUseSphericalLimits=true
    simulationSpace=RootRelative
    alphaBoolBlend {blendOption=Linear}
    boundBone {boneName=dyn_bow_ribbon_01_l}
    chainEnd {boneName=dyn_bow_ribbon_02_l}
    componentAppliedLinearAccClamp=<2000 2000 2000>
    componentLinearAccScale=<0.3 0.3 0.3>
    componentLinearVelScale=<0.3 0.3 0.3>
    gravityOverride {z=-980}
    physicsBodyDefinitions [
     {
      boundBone {boneName=dyn_bow_ribbon_01_l}
      boxExtents=<20 20 20>
      constraintSetup {angularLimitsMax=<0 20 20>, angularLimitsMin=<0 -20 0>, angularTarget=<0 0 0>}
      localJointOffset=<0.5 0 0>
     }
     {
      boundBone {boneName=dyn_bow_ribbon_02_l}
      boxExtents=<20 20 20>
      constraintSetup {angularLimitsMax=<0 20 20>, angularLimitsMin=<0 -20 0>, angularTarget=<0 0 0>}
      localJointOffset=<0.5 0 0>
     }
    ]
    planarLimits [{drivingBone {boneName=thigh_twist_01_r}, planeTransform {location=<0 -6.6679 -1.192>, rotation=<rot -0.8075 -4.522 -100.104>}}]
    relativeSpaceBone {boneName=root}
   }
   n7 = Anim Dynamics - Bone: dyn_fragment_right_r_01(ComponentPose=@n6, Alpha=1, ExternalForce="0,0,0") {
    angularDampingOverride=0.9
    angularSpringConstant=50
    bAngularSpring=true
    bOverrideAngularDamping=true
    bUseGravityOverride=true
    bUseSphericalLimits=true
    simulationSpace=RootRelative
    alphaBoolBlend {blendOption=Linear}
    boundBone {boneName=dyn_fragment_right_r_01}
    componentAppliedLinearAccClamp=<2000 2000 2000>
    componentLinearAccScale=<0.3 0.3 0.3>
    componentLinearVelScale=<0.3 0.3 0.3>
    gravityOverride {z=-980}
    physicsBodyDefinitions [{
      boundBone {boneName=dyn_fragment_right_r_01}
      boxExtents=<5 5 5>
      constraintSetup {angularLimitsMax=<0 30 10>, angularLimitsMin=<0 -30 -10>, angularTarget=<0 0 0>}
      localJointOffset=<-0.5 0 0>
     }]
    relativeSpaceBone {boneName=root}
    sphericalLimits [{limitRadius=8, drivingBone {boneName=thigh_r}, sphereLocalOffset=<22.7068 0.7723 0>}]
   }
   n8 = Anim Dynamics - Bone: dyn_fragment_down_l_01(ComponentPose=@n7, Alpha=1, ExternalForce="0,0,0") {
    angularDampingOverride=0.9
    angularSpringConstant=50
    bAngularSpring=true
    bOverrideAngularDamping=true
    bUseGravityOverride=true
    bUseSphericalLimits=true
    simulationSpace=RootRelative
    alphaBoolBlend {blendOption=Linear}
    boundBone {boneName=dyn_fragment_down_l_01}
    componentAppliedLinearAccClamp=<2000 2000 2000>
    componentLinearAccScale=<0.3 0.3 0.3>
    componentLinearVelScale=<0.3 0.3 0.3>
    gravityOverride {z=-980}
    physicsBodyDefinitions [{
      boundBone {boneName=dyn_fragment_down_l_01}
      boxExtents=<5 5 5>
      constraintSetup {angularLimitsMax=<0 30 0>, angularLimitsMin=<0 -30 0>, angularTarget=<0 0 0>}
      localJointOffset=<-0.5 0 0>
     }]
    relativeSpaceBone {boneName=root}
    sphericalLimits [{limitRadius=8, drivingBone {boneName=thigh_r}, sphereLocalOffset=<22.7068 0.7723 -0.3603>}]
   }
   n9 = "Anim Dynamics - Chain: dyn_laceIn01_l -> dyn_laceIn02_l"(ComponentPose=@n8, Alpha=1, ExternalForce="0,0,0") {
    angularDampingOverride=0.99
    angularSpringConstant=350
    bAngularSpring=true
    bChain=true
    bOverrideAngularDamping=true
    bUseGravityOverride=true
    bUseSphericalLimits=true
    simulationSpace=RootRelative
    alphaBoolBlend {blendOption=Linear}
    boundBone {boneName=dyn_laceIn01_l}
    chainEnd {boneName=dyn_laceIn02_l}
    componentAppliedLinearAccClamp=<2000 2000 2000>
    componentLinearAccScale=<0.2 0.2 0.2>
    componentLinearVelScale=<0.2 0.2 0.2>
    gravityOverride {z=-980}
    physicsBodyDefinitions [
     {
      boundBone {boneName=dyn_laceIn01_l}
      boxExtents=<8 8 8>
      constraintSetup {angularLimitsMax=<0 30 30>, angularLimitsMin=<0 -30 -30>, angularTarget=<0 0 0>}
      localJointOffset=<0.5 0.5 0>
     }
     {
      boundBone {boneName=dyn_laceIn02_l}
      boxExtents=<8 8 8>
      constraintSetup {angularLimitsMax=<0 20 20>, angularLimitsMin=<0 -20 -20>, angularTarget=<0 0 0>}
      localJointOffset=<0.5 0.5 0>
     }
    ]
    planarLimits [{drivingBone {boneName=ball_l}, planeTransform {location=<-0.0159 -0.6276 4.3172>, rotation=<rot -2.2542 0.2248 -14.3646>}}]
    relativeSpaceBone {boneName=root}
   }
   n10 = "Anim Dynamics - Chain: dyn_laceOut01_l -> dyn_laceOut02_l"(ComponentPose=@n9, Alpha=1, ExternalForce="0,0,0") {
    angularDampingOverride=0.99
    angularSpringConstant=350
    bAngularSpring=true
    bChain=true
    bOverrideAngularDamping=true
    bUseGravityOverride=true
    simulationSpace=RootRelative
    alphaBoolBlend {blendOption=Linear}
    boundBone {boneName=dyn_laceOut01_l}
    chainEnd {boneName=dyn_laceOut02_l}
    componentAppliedLinearAccClamp=<2000 2000 2000>
    componentLinearAccScale=<0.2 0.2 0.2>
    componentLinearVelScale=<0.2 0.2 0.2>
    gravityOverride {z=-980}
    physicsBodyDefinitions [
     {
      boundBone {boneName=dyn_laceOut01_l}
      boxExtents=<8 8 8>
      constraintSetup {angularLimitsMax=<0 30 30>, angularLimitsMin=<0 -30 -30>, angularTarget=<0 0 0>}
      localJointOffset=<0.5 0.5 0>
     }
     {
      boundBone {boneName=dyn_laceOut02_l}
      boxExtents=<8 8 8>
      constraintSetup {angularLimitsMax=<0 20 20>, angularLimitsMin=<0 -20 -20>, angularTarget=<0 0 0>}
      localJointOffset=<0.5 0.5 0>
     }
    ]
    planarLimits [{drivingBone {boneName=ball_l}, planeTransform {location=<-8.4973 -8.9297 -2.0297>, rotation=<rot -10.0125 -179.2267 154.6398>}}]
    relativeSpaceBone {boneName=root}
   }
   n11 = "Anim Dynamics - Chain: dyn_laceOut01_r -> dyn_laceOut02_r"(ComponentPose=@n10, Alpha=1, ExternalForce="0,0,0") {
    angularDampingOverride=0.99
    angularSpringConstant=350
    bAngularSpring=true
    bChain=true
    bOverrideAngularDamping=true
    bUseGravityOverride=true
    bUseSphericalLimits=true
    simulationSpace=RootRelative
    alphaBoolBlend {blendOption=Linear}
    boundBone {boneName=dyn_laceOut01_r}
    chainEnd {boneName=dyn_laceOut02_r}
    componentAppliedLinearAccClamp=<2000 2000 2000>
    componentLinearAccScale=<0.2 0.2 0.2>
    componentLinearVelScale=<0.2 0.2 0.2>
    gravityOverride {z=-980}
    physicsBodyDefinitions [
     {
      boundBone {boneName=dyn_laceOut01_r}
      boxExtents=<8 8 8>
      constraintSetup {angularLimitsMax=<0 30 30>, angularLimitsMin=<0 -30 -30>, angularTarget=<0 0 0>}
      localJointOffset=<0.5 0.5 0>
     }
     {
      boundBone {boneName=dyn_laceOut02_r}
      boxExtents=<8 8 8>
      constraintSetup {angularLimitsMax=<0 30 30>, angularLimitsMin=<0 -30 -30>, angularTarget=<0 0 0>}
      localJointOffset=<0.5 0.5 0>
     }
    ]
    planarLimits [{drivingBone {boneName=ball_r}, planeTransform {location=<6.7868 7.8824 2.1657>, rotation=<rot -2.2542 0.2245 9.8188>}}]
    relativeSpaceBone {boneName=root}
   }
   n12 = "Anim Dynamics - Chain: dyn_laceIn01_r -> dyn_laceIn02_r"(ComponentPose=@n11, Alpha=1, ExternalForce="0,0,0") {
    angularDampingOverride=0.99
    angularSpringConstant=350
    bAngularSpring=true
    bChain=true
    bOverrideAngularDamping=true
    bUseGravityOverride=true
    simulationSpace=RootRelative
    alphaBoolBlend {blendOption=Linear}
    boundBone {boneName=dyn_laceIn01_r}
    chainEnd {boneName=dyn_laceIn02_r}
    componentAppliedLinearAccClamp=<2000 2000 2000>
    componentLinearAccScale=<0.2 0.2 0.2>
    componentLinearVelScale=<0.2 0.2 0.2>
    gravityOverride {z=-980}
    physicsBodyDefinitions [
     {
      boundBone {boneName=dyn_laceIn01_r}
      boxExtents=<8 8 8>
      constraintSetup {angularLimitsMax=<0 60 60>, angularLimitsMin=<0 -60 -60>, angularTarget=<0 0 0>}
      localJointOffset=<0.5 0.5 0>
     }
     {
      boundBone {boneName=dyn_laceIn02_r}
      boxExtents=<8 8 8>
      constraintSetup {angularLimitsMax=<0 20 20>, angularLimitsMin=<0 -20 -20>, angularTarget=<0 0 0>}
      localJointOffset=<0.5 0.5 0>
     }
    ]
    planarLimits [{drivingBone {boneName=ball_r}, planeTransform {location=<12.3878 9.3227 -2.4639>, rotation=<rot 1.0512 -177.5996 -171.2908>}}]
    relativeSpaceBone {boneName=root}
   }
   n13 = Get Gravity Override()
   n14 = Anim Dynamics - Bone: dyn_ribbon_01_bk!(ComponentPose=@n12, Alpha=1, GravityOverride:Vector=@n13, ExternalForce="0,0,0") -> (Pose) {
    angularDampingOverride=0.9
    angularSpringConstant=350
    bAngularSpring=true
    bOverrideAngularDamping=true
    bUseGravityOverride=true
    bUseSphericalLimits=true
    simulationSpace=RootRelative
    alphaBoolBlend {blendOption=Linear}
    boundBone {boneName=dyn_ribbon_01_bk}
    componentAppliedLinearAccClamp=<2000 2000 2000>
    componentLinearAccScale=<0.5 0.5 0.5>
    componentLinearVelScale=<0.5 0.5 0.5>
    gravityOverride {z=-980}
    physicsBodyDefinitions [{
      boundBone {boneName=dyn_ribbon_01_bk}
      boxExtents=<10 10 10>
      constraintSetup {angularLimitsMax=<0 10 20>, angularLimitsMin=<0 -10 -30>, angularTarget=<0 0 0>}
      localJointOffset=<0.5 0 0>
     }]
    relativeSpaceBone {boneName=root}
    sphericalLimits [{limitRadius=13.594, drivingBone {boneName=pelvis}, sphereLocalOffset=<0 0 0>}]
   }
   n15 = Anim Dynamics - Bone: dyn_ribbon_02_bk!(ComponentPose=@n14, Alpha=1, GravityOverride:Vector=@n13, ExternalForce="0,0,0") -> (Pose) {
    angularDampingOverride=0.9
    angularSpringConstant=350
    bAngularSpring=true
    bOverrideAngularDamping=true
    bUseGravityOverride=true
    bUseSphericalLimits=true
    simulationSpace=RootRelative
    alphaBoolBlend {blendOption=Linear}
    boundBone {boneName=dyn_ribbon_02_bk}
    componentAppliedLinearAccClamp=<2000 2000 2000>
    componentLinearAccScale=<0.5 0.5 0.5>
    componentLinearVelScale=<0.5 0.5 0.5>
    gravityOverride {z=-980}
    physicsBodyDefinitions [{
      boundBone {boneName=dyn_ribbon_02_bk}
      boxExtents=<10 10 10>
      constraintSetup {angularLimitsMax=<0 15 20>, angularLimitsMin=<0 -15 -40>, angularTarget=<0 0 0>}
      localJointOffset=<0.5 0 0>
     }]
    relativeSpaceBone {boneName=root}
    sphericalLimits [{limitRadius=13.594, drivingBone {boneName=pelvis}, sphereLocalOffset=<0 0 0>}]
   }
   n16 = Anim Dynamics - Bone: dyn_ribbon_03_bk!(ComponentPose=@n15, Alpha=1, GravityOverride:Vector=@n13, ExternalForce="0,0,0") -> (Pose) {
    angularDampingOverride=0.9
    angularSpringConstant=350
    bAngularSpring=true
    bOverrideAngularDamping=true
    bUseGravityOverride=true
    bUseSphericalLimits=true
    simulationSpace=RootRelative
    alphaBoolBlend {blendOption=Linear}
    boundBone {boneName=dyn_ribbon_03_bk}
    componentAppliedLinearAccClamp=<2000 2000 2000>
    componentLinearAccScale=<0.5 0.5 0.5>
    componentLinearVelScale=<0.5 0.5 0.5>
    gravityOverride {z=-980}
    physicsBodyDefinitions [{
      boundBone {boneName=dyn_ribbon_03_bk}
      boxExtents=<10 10 10>
      constraintSetup {angularLimitsMax=<0 20 20>, angularLimitsMin=<0 -20 -50>, angularTarget=<0 0 0>}
      localJointOffset=<0.5 0 0>
     }]
    relativeSpaceBone {boneName=root}
    sphericalLimits [{limitRadius=13.594, drivingBone {boneName=pelvis}, sphereLocalOffset=<0 0 0>}]
   }
   n17 = "Transform (Modify) Bone - Bone: dyn_torsoFlap_back_Base"!(ComponentPose=@n16, bAlphaBoolEnabled:bool=True, Translation="0,0,0", Rotation="0.058,45.744,-0.003", Scale="1,1,1") -> (Pose) {
    alphaInputType=Bool
    rotationMode=BMM_Additive
    rotationSpace=BCS_BoneSpace
    translationMode=BMM_Additive
    translationSpace=BCS_BoneSpace
    alphaBoolBlend {blendOption=Linear}
    boneToModify {boneName=dyn_torsoFlap_back_Base}
    rotation=<rot 0.058 45.744 -0.003>
   }
   n18 = "Transform (Modify) Bone - Bone: dyn_torsoFlap_back_Base"(ComponentPose=@n17, bAlphaBoolEnabled=True, Translation="0,1,0", Rotation="0,15,0", Scale="1,1,1") {
    alphaInputType=Bool
    rotationMode=BMM_Additive
    rotationSpace=BCS_BoneSpace
    translationMode=BMM_Additive
    translationSpace=BCS_BoneSpace
    alphaBoolBlend {blendInTime=0.5, blendOption=CubicInOut, blendOutTime=0.5}
    boneToModify {boneName=dyn_torsoFlap_back_Base}
    rotation {yaw=15}
    translation {y=1}
   }
   n19 = "Anim Dynamics - Chain: dyn_ribbon_01_bk -> dyn_ribbon_03_bk"!(ComponentPose=@n18, bAlphaBoolEnabled:bool=True, GravityOverride:Vector=Get Gravity Override(), ExternalForce="0,0,0") -> (Pose) {
    alphaInputType=Bool
    angularDampingOverride=0.9
    angularSpringConstant=350
    bAngularSpring=true
    bChain=true
    bOverrideAngularDamping=true
    bUseGravityOverride=true
    bUseSphericalLimits=true
    simulationSpace=RootRelative
    alphaBoolBlend {blendOption=Linear}
    boundBone {boneName=dyn_ribbon_01_bk}
    chainEnd {boneName=dyn_ribbon_03_bk}
    componentAppliedLinearAccClamp=<2000 2000 2000>
    componentLinearAccScale=<0.3 0.3 0.3>
    componentLinearVelScale=<0.3 0.3 0.3>
    gravityOverride {z=-980}
    physicsBodyDefinitions [
     {
      boundBone {boneName=dyn_ribbon_01_bk}
      boxExtents=<20 20 20>
      constraintSetup {angularLimitsMax=<0 10 30>, angularLimitsMin=<0 -10 -60>, angularTarget=<0 0 0>}
      localJointOffset=<0.5 0 0>
     }
     {
      boundBone {boneName=dyn_ribbon_02_bk}
      boxExtents=<20 20 20>
      constraintSetup {angularLimitsMax=<0 10 30>, angularLimitsMin=<0 -10 -30>, angularTarget=<0 0 0>}
      localJointOffset=<0.5 0 0>
     }
     {
      boundBone {boneName=dyn_ribbon_03_bk}
      boxExtents=<20 20 20>
      constraintSetup {angularLimitsMax=<0 10 30>, angularLimitsMin=<0 -10 -30>, angularTarget=<0 0 0>}
      localJointOffset=<0.5 0 0>
     }
    ]
    relativeSpaceBone {boneName=root}
    sphericalLimits [
     {limitRadius=13.594, drivingBone {boneName=pelvis}, sphereLocalOffset=<-0.8821 -0.0176 0.3138>}
     {limitRadius=10, drivingBone {boneName=spine_01}, sphereLocalOffset=<0 0 0>}
    ]
   }
   n20 = "Transform (Modify) Bone - Bone: dyn_torsoFlap_back_Base"(ComponentPose=@n19, bAlphaBoolEnabled=True, Translation="0,0,0", Rotation="0,-38.098,0", Scale="1,1,1") {
    alphaInputType=Bool
    rotationMode=BMM_Additive
    rotationSpace=BCS_BoneSpace
    translationMode=BMM_Additive
    translationSpace=BCS_BoneSpace
    alphaBoolBlend {blendInTime=0.5, blendOption=CubicInOut, blendOutTime=0.5}
    boneToModify {boneName=dyn_torsoFlap_back_Base}
    rotation {yaw=-38.098}
   }
   n21 = "Transform (Modify) Bone - Bone: dyn_ribbon_01_bk"!(ComponentPose=@n20, bAlphaBoolEnabled:bool=True, Translation="0,0,0", Rotation="0,30,0", Scale="1,1,1") -> (Pose) {
    alphaInputType=Bool
    rotationMode=BMM_Additive
    rotationSpace=BCS_BoneSpace
    translationMode=BMM_Additive
    translationSpace=BCS_BoneSpace
    alphaBoolBlend {blendInTime=0.5, blendOption=CubicInOut, blendOutTime=0.5}
    boneToModify {boneName=dyn_ribbon_01_bk}
    rotation {yaw=30}
   }
   Output Pose(Result=Component To Local(ComponentPose=@n21)) {layerGroup=None}
   Save cached pose 'SavedPose15805'(Pose=Copy Pose From Mesh() {bUseAttachedParent=true}) {cacheName=SavedPose15805}
  ]}
 EventGraph {nodes [
   Event BlueprintUpdateAnimation()
   > n1 = Gravity Override State Function DEPRECATED(CharacterPartAnimInstance=Self-Reference(), GameTimeInSeconds=GetGameTimeInSeconds(), GravityOverrideParams=Get Gravity Override Params())
   > Set Gravity Override(Gravity Override=@n1)
   n2 = CheckForBackbling()
   > Set is Wearing Backblings(is Wearing Backblings:bool=OR Boolean(A=@n2, B=@n2.bIsBackblingACape)) -> (Output_Get:bool)
   n3 = Get bIsJumping()
   n4 = Get bIsFalling()
   > Set JumpFall(JumpFall:bool=AND Boolean(A=Get is Wearing Backblings() -> (is Wearing Backblings:bool), B=OR Boolean(A=@n3, B=@n4))) -> (Output_Get:bool)
   n5 = Get bIsSkydiveFloating()
   n6 = Get bIsParachuteOpened()
   > "Set is WearingBackpack&&SkyFloating"("is WearingBackpack&&SkyFloating":bool=AND Boolean(A=Get is Wearing Backblings(), B=OR Boolean(A=@n5, B=@n6))) -> (Output_Get:bool)
   > Set Not_playing_Emote(Not_playing_Emote:bool=NOT Boolean(A=Get bIsPlayingEmote())) -> (Output_Get:bool)
   TryGetPawnOwner()
  ]}
]
variables [
 Gravity Override {type=Vector}
 Gravity Override Params {type=GravityOverrideParamsStruct}
 is Wearing Backblings {type=boolean}
 is Wearing Cape {type=boolean}
 is Wearing Cape_on_SkyDiving {type=boolean}
 "is WearingBackpack&&SkyFloating" {type=boolean}
 JumpFall {type=boolean}
 Not_playing_Emote {type=boolean}
]
