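# Physics asset of capsule colliders for the female medium Cardboard Crew Holiday body, named for cape collision.
# Capsules cover the head, neck, spine, pelvis, arms and legs, with two per calf.
# Left, right and back torso-flap bones carry capsules too, each chained through a tiny base capsule to spine_03.
# Constraints follow the skeleton with swing and twist free; the right elbow constraint appears three times.
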PhysicsAsset /BRCosmetics/Characters/Player/Female/Medium/Bodies/F_MED_Cardboard_Crew_Holiday/Meshes/F_MED_Cardboard_Crew_Holiday_CapeColliders
bodies {
 calf_l [
  {shapeType=Capsule, length=10.0137, radius=5.951, shapeName=calf_l_capsule, center=<-9.832 -3.6012 0>, rotation=<rot 90 159.444 139.3273>}
  {shapeType=Capsule, length=32.8537, radius=7.4362, shapeName=calf_l_capsule, center=<-29.1877 -4.9056 0>, rotation=<rot 90 180 -180>}
 ]
 calf_r [
  {shapeType=Capsule, length=10.0137, radius=5.951, shapeName=calf_r_capsule, center=<9.4688 3.4682 0>, rotation=<rot 90 172.875 152.7583>}
  {shapeType=Capsule, length=32.8537, radius=7.4362, shapeName=calf_r_capsule, center=<28.7093 2.2718 0>, rotation=<rot 90 180 -180>}
 ]
 clavicle_l [{shapeType=Capsule, length=23.9223, radius=6.0777, shapeName=clavicle_l_capsule, rotation=<rot -90 92.894 -92.8941>}]
 clavicle_r [{shapeType=Capsule, length=23.9223, radius=6.0777, shapeName=clavicle_r_capsule, rotation=<rot -90 92.894 -92.8941>}]
 dyn_torsoFlap_Base_l [{shapeType=Capsule, length=0, radius=0.5, shapeName=dyn_torsoFlap_Base_l_capsule}]
 dyn_torsoFlap_Base_r [{shapeType=Capsule, length=0, radius=0.5, shapeName=dyn_torsoFlap_Base_r_capsule}]
 dyn_torsoFlap_back [{shapeType=Capsule, length=15.5, radius=5, shapeName=dyn_torsoFlap_back_capsule, center=<6.841 0.0002 0>, rotation=<rot 0.0001 0.0013 0>}]
 dyn_torsoFlap_back_Base [{shapeType=Capsule, length=0, radius=0.5, shapeName=dyn_torsoFlap_back_Base_capsule}]
 dyn_torsoFlap_l [{shapeType=Capsule, length=10.5, radius=8.5, shapeName=dyn_torsoFlap_l_capsule, center=<5.1774 0 0.0001>, rotation=<rot 0.001 0 -0.0004>}]
 dyn_torsoFlap_r [{
   shapeType=Capsule
   length=10.5
   radius=8.5
   shapeName=dyn_torsoFlap_r_capsule
   center=<5.1774 0 -0.0001>
   rotation=<rot -0.0011 -0.0001 0.0004>
  }]
 foot_l [{
   shapeType=Capsule
   length=12.5982
   radius=6.5127
   shapeName=foot_l_capsule
   center=<-2.2072 6.0534 -0.3367>
   rotation=<rot 0.0388 1.4701 -89.3982>
  }]
 foot_r [{
   shapeType=Capsule
   length=12.5982
   radius=6.5127
   shapeName=foot_r_capsule
   center=<2.2072 -6.0534 0.3367>
   rotation=<rot 0.0388 1.4701 90.6018>
  }]
 head [{shapeType=Capsule, length=16.7918, radius=13.2017, shapeName=head_capsule, rotation=<rot 90 177.4695 177.4696>}]
 lowerarm_l [{shapeType=Capsule, length=24.889, radius=5.111, shapeName=lowerarm_l_capsule, center=<8.5111 0 0>, rotation=<rot -90 179.9992 -179.9992>}]
 lowerarm_r [{shapeType=Capsule, length=24.889, radius=5.111, shapeName=lowerarm_r_capsule, center=<-10.6925 0 0>, rotation=<rot -90 6.7214 -6.7214>}]
 neck_02 [{shapeType=Capsule, length=12.0664, radius=6.8979, shapeName=neck_02_capsule, rotation=<rot 90 -144.7356 -144.7355>}]
 pelvis [{shapeType=Capsule, length=14.5536, radius=14.878, shapeName=pelvis_capsule, center=<0 -2.1721 0>}]
 spine_01 [{shapeType=Capsule, length=9.1076, radius=15.4447, shapeName=spine_01_capsule, center=<0 -2.1721 0>}]
 spine_02 [{shapeType=Capsule, length=7.6412, radius=14.892, shapeName=spine_02_capsule, center=<0 -2.1721 0>}]
 spine_03 [{shapeType=Capsule, length=12.8759, radius=13.644, shapeName=spine_03_capsule, center=<0 -2.1721 0>}]
 spine_04 [{shapeType=Capsule, length=3.6804, radius=15.0781, shapeName=spine_04_capsule}]
 spine_05 [{shapeType=Capsule, length=15.4476, radius=11.8004, shapeName=spine_05_capsule}]
 thigh_l [{shapeType=Capsule, length=20.3762, radius=9.6238, shapeName=thigh_l_capsule, center=<-23.8647 0 0>, rotation=<rot -90 -179.9999 179.9999>}]
 thigh_r [{shapeType=Capsule, length=20.3762, radius=9.6238, shapeName=thigh_r_capsule, center=<26.8899 0 0>, rotation=<rot -90 180 180>}]
 upperarm_l [{shapeType=Capsule, length=24.5261, radius=5.4739, shapeName=upperarm_l_capsule, center=<6.2274 0 0>, rotation=<rot -90 180 -180>}]
 upperarm_r [{shapeType=Capsule, length=24.5261, radius=5.4739, shapeName=upperarm_r_capsule, center=<-6.9733 0 0>, rotation=<rot -90 179.9926 -179.9926>}]
}
constraints [
 {
  bone1Name=spine_01
  bone2Name=pelvis
  swing1LimitDegrees=45
  swing1Motion=Free
  swing2LimitDegrees=45
  swing2Motion=Free
  twistLimitDegrees=45
  twistMotion=Free
 }
 {
  bone1Name=spine_02
  bone2Name=spine_01
  swing1LimitDegrees=45
  swing1Motion=Free
  swing2LimitDegrees=45
  swing2Motion=Free
  twistLimitDegrees=45
  twistMotion=Free
 }
 {
  bone1Name=spine_03
  bone2Name=spine_02
  swing1LimitDegrees=45
  swing1Motion=Free
  swing2LimitDegrees=45
  swing2Motion=Free
  twistLimitDegrees=45
  twistMotion=Free
 }
 {
  bone1Name=spine_04
  bone2Name=spine_03
  swing1LimitDegrees=45
  swing1Motion=Free
  swing2LimitDegrees=45
  swing2Motion=Free
  twistLimitDegrees=45
  twistMotion=Free
 }
 {
  bone1Name=spine_05
  bone2Name=spine_04
  swing1LimitDegrees=45
  swing1Motion=Free
  swing2LimitDegrees=45
  swing2Motion=Free
  twistLimitDegrees=45
  twistMotion=Free
 }
 {
  bone1Name=clavicle_l
  bone2Name=spine_05
  swing1LimitDegrees=45
  swing1Motion=Free
  swing2LimitDegrees=45
  swing2Motion=Free
  twistLimitDegrees=45
  twistMotion=Free
 }
 {
  bone1Name=upperarm_l
  bone2Name=clavicle_l
  swing1LimitDegrees=45
  swing1Motion=Free
  swing2LimitDegrees=45
  swing2Motion=Free
  twistLimitDegrees=45
  twistMotion=Free
 }
 {
  bone1Name=clavicle_r
  bone2Name=spine_05
  swing1LimitDegrees=45
  swing1Motion=Free
  swing2LimitDegrees=45
  swing2Motion=Free
  twistLimitDegrees=45
  twistMotion=Free
 }
 {
  bone1Name=upperarm_r
  bone2Name=clavicle_r
  swing1LimitDegrees=45
  swing1Motion=Free
  swing2LimitDegrees=45
  swing2Motion=Free
  twistLimitDegrees=45
  twistMotion=Free
 }
 {
  bone1Name=neck_02
  bone2Name=spine_05
  swing1LimitDegrees=45
  swing1Motion=Free
  swing2LimitDegrees=45
  swing2Motion=Free
  twistLimitDegrees=45
  twistMotion=Free
 }
 {
  bone1Name=head
  bone2Name=neck_02
  swing1LimitDegrees=45
  swing1Motion=Free
  swing2LimitDegrees=45
  swing2Motion=Free
  twistLimitDegrees=45
  twistMotion=Free
 }
 {
  bone1Name=thigh_l
  bone2Name=pelvis
  swing1LimitDegrees=45
  swing1Motion=Free
  swing2LimitDegrees=45
  swing2Motion=Free
  twistLimitDegrees=45
  twistMotion=Free
 }
 {
  bone1Name=calf_l
  bone2Name=thigh_l
  swing1LimitDegrees=45
  swing1Motion=Free
  swing2LimitDegrees=45
  swing2Motion=Free
  twistLimitDegrees=45
  twistMotion=Free
 }
 {
  bone1Name=foot_l
  bone2Name=calf_l
  swing1LimitDegrees=45
  swing1Motion=Free
  swing2LimitDegrees=45
  swing2Motion=Free
  twistLimitDegrees=45
  twistMotion=Free
 }
 {
  bone1Name=thigh_r
  bone2Name=pelvis
  swing1LimitDegrees=45
  swing1Motion=Free
  swing2LimitDegrees=45
  swing2Motion=Free
  twistLimitDegrees=45
  twistMotion=Free
 }
 {
  bone1Name=calf_r
  bone2Name=thigh_r
  swing1LimitDegrees=45
  swing1Motion=Free
  swing2LimitDegrees=45
  swing2Motion=Free
  twistLimitDegrees=45
  twistMotion=Free
 }
 {
  bone1Name=foot_r
  bone2Name=calf_r
  swing1LimitDegrees=45
  swing1Motion=Free
  swing2LimitDegrees=45
  swing2Motion=Free
  twistLimitDegrees=45
  twistMotion=Free
 }
 {
  bone1Name=lowerarm_r
  bone2Name=upperarm_r
  swing1LimitDegrees=45
  swing1Motion=Free
  swing2LimitDegrees=45
  swing2Motion=Free
  twistLimitDegrees=45
  twistMotion=Free
 }
 {
  bone1Name=lowerarm_l
  bone2Name=upperarm_l
  swing1LimitDegrees=45
  swing1Motion=Free
  swing2LimitDegrees=45
  swing2Motion=Free
  twistLimitDegrees=45
  twistMotion=Free
 }
 {
  bone1Name=lowerarm_r
  bone2Name=upperarm_r
  swing1LimitDegrees=45
  swing1Motion=Free
  swing2LimitDegrees=45
  swing2Motion=Free
  twistLimitDegrees=45
  twistMotion=Free
 }
 {
  bone1Name=lowerarm_r
  bone2Name=upperarm_r
  swing1LimitDegrees=45
  swing1Motion=Free
  swing2LimitDegrees=45
  swing2Motion=Free
  twistLimitDegrees=45
  twistMotion=Free
 }
 {
  bone1Name=dyn_torsoFlap_l
  bone2Name=dyn_torsoFlap_Base_l
  swing1LimitDegrees=45
  swing1Motion=Free
  swing2LimitDegrees=45
  swing2Motion=Free
  twistLimitDegrees=45
  twistMotion=Free
 }
 {
  bone1Name=dyn_torsoFlap_r
  bone2Name=dyn_torsoFlap_Base_r
  swing1LimitDegrees=45
  swing1Motion=Free
  swing2LimitDegrees=45
  swing2Motion=Free
  twistLimitDegrees=45
  twistMotion=Free
 }
 {
  bone1Name=dyn_torsoFlap_back
  bone2Name=dyn_torsoFlap_back_Base
  swing1LimitDegrees=45
  swing1Motion=Free
  swing2LimitDegrees=45
  swing2Motion=Free
  twistLimitDegrees=45
  twistMotion=Free
 }
 {
  bone1Name=dyn_torsoFlap_Base_l
  bone2Name=spine_03
  swing1LimitDegrees=45
  swing1Motion=Free
  swing2LimitDegrees=45
  swing2Motion=Free
  twistLimitDegrees=45
  twistMotion=Free
 }
 {
  bone1Name=dyn_torsoFlap_Base_r
  bone2Name=spine_03
  swing1LimitDegrees=45
  swing1Motion=Free
  swing2LimitDegrees=45
  swing2Motion=Free
  twistLimitDegrees=45
  twistMotion=Free
 }
 {
  bone1Name=dyn_torsoFlap_back_Base
  bone2Name=spine_03
  swing1LimitDegrees=45
  swing1Motion=Free
  swing2LimitDegrees=45
  swing2Motion=Free
  twistLimitDegrees=45
  twistMotion=Free
 }
]
