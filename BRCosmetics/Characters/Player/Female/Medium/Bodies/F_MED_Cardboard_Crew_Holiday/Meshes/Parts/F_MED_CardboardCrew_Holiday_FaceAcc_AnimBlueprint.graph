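# Post-process style AnimBlueprint for the face accessory of the holiday Cardboard Crew female cosmetic, animating its head flaps, bow and ribbons.
# The pose copied from the attached parent mesh feeds a Control Rig driven by targeting, crouch-targeting, parachute and skydive-floating flags.
# Additive flap offsets then lead into RigidBody simulation and chained AnimDynamics for bow and ribbon bones, constrained by planar and spherical limits.
# Final additive rotations adjust the ribbon base and back-right flap.
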AnimBlueprint /BRCosmetics/Characters/Player/Female/Medium/Bodies/F_MED_Cardboard_Crew_Holiday/Meshes/Parts/F_MED_CardboardCrew_Holiday_FaceAcc_AnimBlueprint
target_skeleton=/Game/Characters/Player/Female/Medium/Bodies/F_MED_Cardboard_Crew_Holiday/Meshes/Parts/F_MED_CardboardCrew_Holiday_FaceAcc_Skeleton
graphs [
 AnimGraph {nodes [
   Anim Dynamics - Bone: dyn_head_bowRibbon_Base!(ComponentPose, Alpha=1, GravityOverride:Vector="0,0,-980", ExternalForce="0,0,0") -> (Pose) {
    angularDampingOverride=0.9
    angularSpringConstant=350
    bAngularSpring=true
    bOverrideAngularDamping=true
    bUseGravityOverride=true
    bUseSphericalLimits=true
    simulationSpace=RootRelative
    alphaBoolBlend {blendOption=Linear}
    boundBone {boneName=dyn_head_bowRibbon_Base}
    componentAppliedLinearAccClamp=<2000 2000 2000>
    componentLinearAccScale=<0.5 0.5 0.5>
    componentLinearVelScale=<0.5 0.5 0.5>
    gravityOverride {z=-980}
    physicsBodyDefinitions [{
      boundBone {boneName=dyn_head_bowRibbon_Base}
      boxExtents=<10 10 10>
      constraintSetup {angularLimitsMax=<0 10 10>, angularLimitsMin=<0 -10 -20>, angularTarget=<0 0 0>}
      localJointOffset=<0.5 0 0>
     }]
    planarLimits [
     {drivingBone {boneName=head}, planeTransform {location=<0.1214 -6.201 -4.9672>, rotation=<rot 2.435 3.3842 -125.6533>}}
     {drivingBone {boneName=dyn_headFlapBL_01}, planeTransform {location=<0 0 0>, rotation=<rot 90 0 0>}}
     {drivingBone {boneName=dyn_headFlapBL_03}, planeTransform {location=<-0.1817 -0.8446 0.0098>, rotation=<rot 90 77.8552 0>}}
    ]
    sphericalLimits [{limitRadius=10, drivingBone {boneName=dyn_head}, sphereLocalOffset=<7.0158 1.1471 1.6916>}]
   }
   n1 = Control Rig!(
    Source=Use cached pose 'SavedPose27371'()
    Alpha
    is_Targeting:bool=Get bIsTargeting()
    isParashuteOpened:bool=Get bIsParachuteOpened()
    TargetingOnCrouch:bool=Get Targeting_on_Crouch() -> (Targeting_on_Crouch:bool)
    isSkyFloating:bool=Get bIsSkydiveFloating()
   ) -> (Pose) {
    alphaBoolBlend {blendOption=Linear}
    controlRigAssetReference {blueprintRigClass=/Game/Characters/Player/Female/Medium/Bodies/F_MED_Cardboard_Crew_Holiday/Meshes/Parts/F_MED_CardboardCrew_Holiday_FaceAcc_CtrlRig.F_MED_CardboardCrew_Holiday_FaceAcc_CtrlRig_C}
   }
   n2 = "Transform (Modify) Bone - Bone: dyn_headFlapBR_01_Base"!(ComponentPose=Local To Component(LocalPose=@n1), bAlphaBoolEnabled:bool=True, Translation="0,0,0", Rotation="0,-50,0", Scale="1,1,1") -> (Pose) {
    alphaInputType=Bool
    rotationMode=BMM_Additive
    rotationSpace=BCS_BoneSpace
    alphaBoolBlend {blendInTime=0.5, blendOption=CubicInOut, blendOutTime=0.5}
    boneToModify {boneName=dyn_headFlapBR_01_Base}
    rotation {yaw=-50}
   }
   n3 = "Transform (Modify) Bone - Bone: dyn_headFlapBR_01_Base"(ComponentPose=@n2, bAlphaBoolEnabled=True, Translation="0,0,-1.926", Rotation="-10,-30,0", Scale="1,1,1") {
    alphaInputType=Bool
    rotationMode=BMM_Additive
    rotationSpace=BCS_BoneSpace
    translationMode=BMM_Additive
    translationSpace=BCS_BoneSpace
    alphaBoolBlend {blendInTime=0.5, blendOption=CubicInOut, blendOutTime=0.5}
    boneToModify {boneName=dyn_headFlapBR_01_Base}
    rotation {pitch=-10, yaw=-30}
    translation {z=-1.926}
   }
   n4 = RigidBody(ComponentPose=@n3) {
    bClampLinearTranslationLimitToRefPose=true
    bOverrideWorldGravity=true
    overridePhysicsAsset=/Game/Characters/Player/Female/Medium/Bodies/F_MED_Cardboard_Crew_Holiday/Meshes/Parts/F_MED_CardboardCrew_Holiday_FaceAcc_Physics
    simulationSpace=BaseBoneSpace
    alphaBoolBlend {blendOption=Linear}
    baseBoneRef {boneName=spine_05}
    componentAppliedLinearAccClamp=<3000 3000 3000>
    componentLinearAccScale=<0.5 0.5 0.5>
    componentLinearVelScale=<0.2 0.2 0.2>
    overrideWorldGravity {z=-980}
   }
   n5 = Anim Dynamics - Bone: dyn_head_bow_l_01(ComponentPose=@n4, Alpha=1, ExternalForce="0,0,0") {
    angularDampingOverride=0.9
    angularSpringConstant=350
    bAngularSpring=true
    bOverrideAngularDamping=true
    bUseGravityOverride=true
    simulationSpace=RootRelative
    alphaBoolBlend {blendOption=Linear}
    boundBone {boneName=dyn_head_bow_l_01}
    componentAppliedLinearAccClamp=<2000 2000 2000>
    componentLinearAccScale=<0.45 0.45 0.45>
    componentLinearVelScale=<0.45 0.45 0.45>
    gravityOverride {z=-980}
    physicsBodyDefinitions [{
      boundBone {boneName=dyn_head_bow_l_01}
      boxExtents=<10 10 10>
      constraintSetup {angularLimitsMax=<0 0 20>, angularLimitsMin=<0 -20 0>, angularTarget=<0 0 0>}
      localJointOffset=<0.5 0 0>
     }]
    planarLimits [{drivingBone {boneName=dyn_head}, planeTransform {location=<0 -4.7885 -5.4197>, rotation=<rot 0 -0.0001 -138.5378>}}]
   }
   n6 = Anim Dynamics - Bone: dyn_head_bow_r_01(ComponentPose=@n5, Alpha=1, ExternalForce="0,0,0") {
    angularDampingOverride=0.9
    angularSpringConstant=350
    bAngularSpring=true
    bOverrideAngularDamping=true
    bUseGravityOverride=true
    simulationSpace=RootRelative
    alphaBoolBlend {blendOption=Linear}
    boundBone {boneName=dyn_head_bow_r_01}
    componentAppliedLinearAccClamp=<2000 2000 2000>
    componentLinearAccScale=<0.45 0.45 0.45>
    componentLinearVelScale=<0.45 0.45 0.45>
    gravityOverride {z=-980}
    physicsBodyDefinitions [{
      boundBone {boneName=dyn_head_bow_r_01}
      boxExtents=<10 10 10>
      constraintSetup {angularLimitsMax=<0 0 0>, angularLimitsMin=<0 -20 -20>, angularTarget=<0 0 0>}
      localJointOffset=<0.5 0 0>
     }]
    planarLimits [{drivingBone {boneName=dyn_head}, planeTransform {location=<0 -4.8501 -5.467>, rotation=<rot 0 0 -138.422>}}]
   }
   n7 = Anim Dynamics - Bone: dyn_head_bowRibbon_l_01!(ComponentPose=@n6, Alpha=1, GravityOverride:Vector="0,0,-980", ExternalForce="0,0,0") -> (Pose) {
    angularDampingOverride=0.9
    angularSpringConstant=350
    bAngularSpring=true
    bOverrideAngularDamping=true
    bUseGravityOverride=true
    bUseSphericalLimits=true
    simulationSpace=RootRelative
    alphaBoolBlend {blendOption=Linear}
    boundBone {boneName=dyn_head_bowRibbon_l_01}
    componentAppliedLinearAccClamp=<2000 2000 2000>
    componentLinearAccScale=<0.5 0.5 0.5>
    componentLinearVelScale=<0.5 0.5 0.5>
    gravityOverride {z=-980}
    physicsBodyDefinitions [{
      boundBone {boneName=dyn_head_bowRibbon_l_01}
      boxExtents=<10 10 10>
      constraintSetup {angularLimitsMax=<0 10 10>, angularLimitsMin=<0 -10 -30>, angularTarget=<0 0 0>}
      localJointOffset=<0.5 0 0>
     }]
    planarLimits [
     {drivingBone {boneName=head}, planeTransform {location=<0.1214 -6.201 -4.9672>, rotation=<rot 2.435 3.3842 -125.6533>}}
     {drivingBone {boneName=dyn_headFlapBL_01}, planeTransform {location=<0 0 0>, rotation=<rot 90 0 0>}}
     {drivingBone {boneName=dyn_headFlapBL_03}, planeTransform {location=<-0.1817 -0.8446 0.0098>, rotation=<rot 90 77.8552 0>}}
    ]
    sphericalLimits [{limitRadius=10, drivingBone {boneName=dyn_head}, sphereLocalOffset=<3.88 0 1.6916>}]
   }
   n8 = Anim Dynamics - Bone: dyn_head_bowRibbon_l_02!(ComponentPose=@n7, Alpha=1, GravityOverride:Vector="0,0,-980", ExternalForce="0,0,0") -> (Pose) {
    angularDampingOverride=0.9
    angularSpringConstant=350
    bAngularSpring=true
    bOverrideAngularDamping=true
    bUseGravityOverride=true
    bUseSphericalLimits=true
    simulationSpace=RootRelative
    alphaBoolBlend {blendOption=Linear}
    boundBone {boneName=dyn_head_bowRibbon_l_02}
    componentAppliedLinearAccClamp=<2000 2000 2000>
    componentLinearAccScale=<0.5 0.5 0.5>
    componentLinearVelScale=<0.5 0.5 0.5>
    gravityOverride {z=-980}
    physicsBodyDefinitions [{
      boundBone {boneName=dyn_head_bowRibbon_l_02}
      boxExtents=<10 10 10>
      constraintSetup {angularLimitsMax=<0 10 5>, angularLimitsMin=<0 -10 -30>, angularTarget=<0 0 0>}
      localJointOffset=<0.5 0 0>
     }]
    planarLimits [
     {drivingBone {boneName=head}, planeTransform {location=<0.1214 -6.201 -4.9672>, rotation=<rot 2.435 3.3842 -125.6533>}}
     {drivingBone {boneName=dyn_headFlapBL_01}, planeTransform {location=<0 0 0>, rotation=<rot 90 0 0>}}
     {drivingBone {boneName=dyn_headFlapBL_03}, planeTransform {location=<-0.1817 -0.8446 0.0098>, rotation=<rot 90 77.8552 0>}}
    ]
    sphericalLimits [{limitRadius=10, drivingBone {boneName=dyn_head}, sphereLocalOffset=<3.88 2.1834 0.4379>}]
   }
   n9 = Anim Dynamics - Bone: dyn_head_bowRibbon_r_01!(ComponentPose=@n8, Alpha=1, GravityOverride:Vector="0,0,-980", ExternalForce="0,0,0") -> (Pose) {
    angularDampingOverride=0.9
    angularSpringConstant=350
    bAngularSpring=true
    bOverrideAngularDamping=true
    bUseGravityOverride=true
    bUseSphericalLimits=true
    simulationSpace=RootRelative
    alphaBoolBlend {blendOption=Linear}
    boundBone {boneName=dyn_head_bowRibbon_r_01}
    componentAppliedLinearAccClamp=<2000 2000 2000>
    componentLinearAccScale=<0.5 0.5 0.5>
    componentLinearVelScale=<0.5 0.5 0.5>
    gravityOverride {z=-980}
    physicsBodyDefinitions [{
      boundBone {boneName=dyn_head_bowRibbon_r_01}
      boxExtents=<10 10 10>
      constraintSetup {angularLimitsMax=<0 10 10>, angularLimitsMin=<0 -10 -20>, angularTarget=<0 0 0>}
      localJointOffset=<0.5 0 0>
     }]
    planarLimits [
     {drivingBone {boneName=head}, planeTransform {location=<0 -3.8624 -0.5146>, rotation=<rot 0 -0.0004 -171.778>}}
     {drivingBone {boneName=dyn_headFlapBL_01}, planeTransform {location=<2.9097 0.5589 -0.0703>, rotation=<rot 89.6149 -179.7524 -179.7536>}}
     {drivingBone {boneName=dyn_headFlapBL_03}, planeTransform {location=<-4.4166 -3.5723 1.5984>, rotation=<rot -62.5264 177.0473 174.5819>}}
    ]
    sphericalLimits [{limitRadius=12.9375, drivingBone {boneName=dyn_head}, sphereLocalOffset=<3.3039 -0.3486 5.7257>}]
   }
   n10 = Anim Dynamics - Bone: dyn_head_bowRibbon_r_02!(ComponentPose=@n9, Alpha=1, GravityOverride:Vector="0,0,-980", ExternalForce="0,0,0") -> (Pose) {
    angularDampingOverride=0.9
    angularSpringConstant=350
    bAngularSpring=true
    bOverrideAngularDamping=true
    bUseGravityOverride=true
    bUseSphericalLimits=true
    simulationSpace=RootRelative
    alphaBoolBlend {blendOption=Linear}
    boundBone {boneName=dyn_head_bowRibbon_r_02}
    componentAppliedLinearAccClamp=<2000 2000 2000>
    componentLinearAccScale=<0.5 0.5 0.5>
    componentLinearVelScale=<0.5 0.5 0.5>
    gravityOverride {z=-980}
    physicsBodyDefinitions [{
      boundBone {boneName=dyn_head_bowRibbon_r_02}
      boxExtents=<10 10 10>
      constraintSetup {angularLimitsMax=<0 10 5>, angularLimitsMin=<0 -10 -30>, angularTarget=<0 0 0>}
      localJointOffset=<0.5 0 0>
     }]
    planarLimits [
     {drivingBone {boneName=head}, planeTransform {location=<0 -3.8624 -0.5146>, rotation=<rot 0 -0.0005 -171.778>}}
     {drivingBone {boneName=dyn_headFlapBL_01}, planeTransform {location=<2.9097 0.5589 -0.0703>, rotation=<rot 89.608 -179.7583 -179.7594>}}
     {drivingBone {boneName=dyn_headFlapBL_03}, planeTransform {location=<-4.4166 -3.5723 1.5984>, rotation=<rot -62.5264 177.0474 174.5818>}}
    ]
    sphericalLimits [{limitRadius=12.9375, drivingBone {boneName=dyn_head}, sphereLocalOffset=<3.3039 -0.3486 5.7257>}]
   }
   n11 = "Transform (Modify) Bone - Bone: dyn_head_bowRibbon_Base"!(ComponentPose=@n10, bAlphaBoolEnabled:bool=True, Translation="0,0,0", Rotation="0,-80,0", Scale="1,1,1") -> (Pose) {
    alphaInputType=Bool
    rotationMode=BMM_Additive
    rotationSpace=BCS_BoneSpace
    alphaBoolBlend {blendOption=Sinusoidal}
    boneToModify {boneName=dyn_head_bowRibbon_Base}
    rotation {yaw=-80}
   }
   n12 = "Transform (Modify) Bone - Bone: dyn_headFlapBR_01"!(ComponentPose=@n11, bAlphaBoolEnabled:bool=True, Translation="0,0,0", Rotation="0,-20,-10", Scale="1,1,1") -> (Pose) {
    alphaInputType=Bool
    rotationMode=BMM_Additive
    rotationSpace=BCS_BoneSpace
    alphaBoolBlend {blendOption=Sinusoidal}
    boneToModify {boneName=dyn_headFlapBR_01}
    rotation {roll=-10, yaw=-20}
   }
   Output Pose(Result=Component To Local(ComponentPose=@n12)) {layerGroup=None}
   Save cached pose 'SavedPose27371'(Pose=Copy Pose From Mesh() {bUseAttachedParent=true}) {cacheName=SavedPose27371}
  ]}
 EventGraph {nodes [
   Event BlueprintUpdateAnimation()
   > Set Targeting_on_Crouch(Targeting_on_Crouch:bool=AND Boolean(A=Get bIsTargeting(), B=Get bIsCrouching())) -> (Output_Get:bool)
   TryGetPawnOwner()
  ]}
]
variables [
 Game Time in Seconds {type=number}
 Gravity Override {type=Vector}
 Gravity Override Params {type=GravityOverrideParamsStruct}
 Targeting_on_Crouch {type=boolean}
]
